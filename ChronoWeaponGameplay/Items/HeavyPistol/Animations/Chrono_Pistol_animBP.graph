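# Animation Blueprint for the Chrono heavy pistol's weapon mesh, posing the hammer to match the gun's state.
# Each update casts the owner to B_Pistol_Chrono_Athena and sets flags from ammo, reload state and the IsFired anim curve.
# The Hammer state machine leaves StartingPose for PulledBack if loaded or Fired if empty, then moves to Fired on the curve and back on reload.
# Its output passes through DefaultSlot, allowing montage playback over the poses.
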AnimBlueprint /ChronoWeaponGameplay/Items/HeavyPistol/Animations/Chrono_Pistol_animBP
target_skeleton=/ChronoWeaponGameplay/Items/HeavyPistol/Meshes/Chrono_Pistol_Skeleton
graphs [
 AnimGraph {nodes [
   AnimGraphNode_StateMachine_0 = Hammer() {is=title AnimGraphNode_StateMachine}
   Output Pose(Result=Slot 'DefaultSlot'(Source=@AnimGraphNode_StateMachine_0)) {layerGroup=None}
  ]}
 Hammer {
  entry=StartingPose
  path=AnimGraph.AnimGraphNode_StateMachine_0.Hammer
  ids=id
  nodes [
   AnimStateNode_1 = PulledBack() {is=title AnimStateNode}
   AnimStateNode_2 = Fired() {is=title AnimStateNode}
   AnimStateNode_3 = StartingPose() {is=title AnimStateNode}
   AnimStateTransitionNode_1 = AnimStateTransitionNode() {crossfadeDuration=0.1}
   AnimStateTransitionNode_2 = AnimStateTransitionNode() {crossfadeDuration=0.1}
   AnimStateTransitionNode_5 = AnimStateTransitionNode()
   AnimStateTransitionNode_6 = AnimStateTransitionNode()
  ]
  links [
   {
    dst=PulledBack
    rule=AnimGraph.AnimGraphNode_StateMachine_0.Hammer.AnimStateTransitionNode_2.Transition
    src=Fired
    via=AnimStateTransitionNode_2
   }
   {
    dst=Fired
    rule=AnimGraph.AnimGraphNode_StateMachine_0.Hammer.AnimStateTransitionNode_1.Transition
    src=PulledBack
    via=AnimStateTransitionNode_1
   }
   {
    dst=Fired
    rule=AnimGraph.AnimGraphNode_StateMachine_0.Hammer.AnimStateTransitionNode_6.Transition
    src=StartingPose
    via=AnimStateTransitionNode_6
   }
   {
    dst=PulledBack
    rule=AnimGraph.AnimGraphNode_StateMachine_0.Hammer.AnimStateTransitionNode_5.Transition
    src=StartingPose
    via=AnimStateTransitionNode_5
   }
  ]
 }
 PulledBack {
  path=AnimGraph.AnimGraphNode_StateMachine_0.Hammer.AnimStateNode_1.PulledBack
  nodes [Output Animation Pose(Result=Sequence Player 'Chrono_Pistol_HammerBack_Pose_gun'() {bLoopAnimation=false, sequence=&/Chrono_Pistol_HammerBack_Pose_gun})]
 }
 Fired {
  path=AnimGraph.AnimGraphNode_StateMachine_0.Hammer.AnimStateNode_2.Fired
  nodes [Output Animation Pose(Result=Sequence Player 'Chrono_HeavyPistol_Fired_Pose_gun'() {bLoopAnimation=false, sequence=&/Chrono_HeavyPistol_Fired_Pose_gun})]
 }
 StartingPose {
  path=AnimGraph.AnimGraphNode_StateMachine_0.Hammer.AnimStateNode_3.StartingPose
  nodes [Output Animation Pose(Result=Sequence Player 'Chrono_Pistol_Base_Pose_gun'() {bLoopAnimation=false, sequence=&/Chrono_Pistol_Base_Pose_gun})]
 }
 Transition {
  path=AnimGraph.AnimGraphNode_StateMachine_0.Hammer.AnimStateTransitionNode_1.Transition
  nodes [Result(bCanEnterTransition=Get IsFired() -> (IsFired:bool))]
 }
 Transition {
  path=AnimGraph.AnimGraphNode_StateMachine_0.Hammer.AnimStateTransitionNode_2.Transition
  nodes [Result(bCanEnterTransition=Get IsReloading() -> (IsReloading:bool))]
 }
 Transition {
  path=AnimGraph.AnimGraphNode_StateMachine_0.Hammer.AnimStateTransitionNode_5.Transition
  nodes [Result(bCanEnterTransition=Get IsLoaded() -> (IsLoaded:bool))]
 }
 Transition {
  path=AnimGraph.AnimGraphNode_StateMachine_0.Hammer.AnimStateTransitionNode_6.Transition
  nodes [Result(bCanEnterTransition=Get IsEmpty() -> (IsEmpty:bool))]
 }
 EventGraph {nodes [
   Event BlueprintUpdateAnimation()
   n1 = GetOwningActor()
   > Is Valid(InputObject=@n1)
   > n2 = Cast To B_Pistol_Chrono_Athena!(execute, Object=@n1) -> (then, CastFailed, AsB Pistol Chrono Athena:B Pistol Chrono Athena Object Reference) {is=title K2Node_DynamicCast}
   n3 = GetRemainingAmmo!(self:Fort Weapon Object Reference=@n2) -> (ReturnValue:int=0)
   > Set IsLoaded(IsLoaded:bool="integer > integer"(A=@n3)) -> (Output_Get:bool)
   > Set IsEmpty(IsEmpty:bool="Equal (Integer)"(A=@n3)) -> (Output_Get:bool)
   > Set IsFired(IsFired:bool="Equal (Float)"(A=GetCurveValue(CurveName=IsFired), B=1)) -> (Output_Get:bool)
   > Set IsReloading(IsReloading:bool=IsReloading!(self:Fort Weapon Object Reference=@n2) -> (ReturnValue:bool)) -> (Output_Get:bool)
  ]}
]
variables [
 IsEmpty {type=boolean}
 IsFired {type=boolean}
 IsLoaded {type=boolean}
 IsReloading {type=boolean}
 WeaponBP {type=B_Pistol_Chrono_Athena_C, default=null}
]
