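# Sound cue for dropping the Chrono heavy pistol.
# A random node picks between two drop wave variants with equal weight.
# It uses shared weapon-drop attenuation and the loot-collect sound class, and overrides concurrency to cap simultaneous plays, blocking new ones once the limit is reached.
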SoundCue /ChronoWeaponGameplay/Items/HeavyPistol/Audio/Chrono_HvyPistol_Drop_Cue
root=Random_1
nodes [
 {id=Random_1, children [WavePlayer_0,WavePlayer_1], properties {weights [1,1]}}
 {id=WavePlayer_0, properties {soundWaveAssetPtr=&/Chrono_HvyPistol_Drop_02}}
 {id=WavePlayer_1, properties {soundWaveAssetPtr=&/Chrono_HvyPistol_Drop_01}}
]
properties {
 attenuationSettings=&Fort_Weapon_Drop_Att
 bOverrideConcurrency=true
 maxDistance=3000
 soundClassObject=&Fort_Loot_Collect
 volumeMultiplier=1.1
 attenuationOverrides {falloffDistance=1000, attenuationShapeExtents {x=0}}
 concurrencyOverrides {maxCount=3, resolutionRule=PreventNew}
}
