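# First-person firing sound cue for the Chrono Heavy Pistol.
# A quality-level root sends its first slot through a modulator to a random pick of five shot waves, while its other three slots share one modulator over a three-wave random pick.
# It uses Fortnite first-person weapon attenuation, a ranged-fire sound class and the shared first-person gunfire concurrency set.
# The debug flag is left on.
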SoundCue /ChronoWeaponGameplay/Items/HeavyPistol/Audio/Chrono_HvyPistol_Fire_1p_Cue
root=QualityLevel_0
nodes [
 {id=Modulator_0, children [Random_0], properties {volumeMax=1.01, volumeMin=0.99}}
 {id=Modulator_1, children [Random_2]}
 {id=QualityLevel_0, children [Modulator_0,Modulator_1,Modulator_1,Modulator_1]}
 {id=Random_0, children [WavePlayer_1,WavePlayer_2,WavePlayer_3,WavePlayer_4,WavePlayer_8], properties {weights [1,1,1,1,1]}}
 {id=Random_2, children [WavePlayer_9, WavePlayer_10, WavePlayer_11], properties {weights [1,1,1]}}
 {id=WavePlayer_1, properties {soundWaveAssetPtr=&/Chrono_HvyPistol_Fire_1p_01}}
 {id=WavePlayer_10, properties {soundWaveAssetPtr=&/Chrono_HvyPistol_Fire_1p_02}}
 {id=WavePlayer_11, properties {soundWaveAssetPtr=&/Chrono_HvyPistol_Fire_1p_03}}
 {id=WavePlayer_2, properties {soundWaveAssetPtr=&Chrono_HvyPistol_Fire_1p_02}}
 {id=WavePlayer_3, properties {soundWaveAssetPtr=&Chrono_HvyPistol_Fire_1p_03}}
 {id=WavePlayer_4, properties {soundWaveAssetPtr=&/Chrono_HvyPistol_Fire_1p_04}}
 {id=WavePlayer_8, properties {soundWaveAssetPtr=&/Chrono_HvyPistol_Fire_1p_05}}
 {id=WavePlayer_9, properties {soundWaveAssetPtr=&Chrono_HvyPistol_Fire_1p_01}}
]
properties {
 attenuationSettings=&Fort_Weapon_Projectile_FirstPerson_Athena_Att
 bDebug=true
 maxDistance=2097152
 soundClassObject=&Fort_WeaponRanged_Fire_1P_PSM
 volumeMultiplier=1
 attenuationOverrides {manualReverbSendLevel=-1.0165, reverbSendMethod=Manual}
 concurrencyOverrides {bLimitToOwner=true, maxCount=3, resolutionRule=StopOldest, volumeScale=0.9}
 concurrencySet [&Weapon_Gunfire_FirstPerson_Conc]
}
