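# Sound cue for a Chrono heavy pistol landing, apparently when the dropped weapon hits the ground.
# A random node picks one of three landed wave variations with equal weight.
# It plays slightly boosted through the loot-collect sound class with weapon-drop attenuation, and a shared weapon-drop concurrency set limits overlapping instances.
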SoundCue /ChronoWeaponGameplay/Items/HeavyPistol/Audio/Chrono_HvyPistol_Landed_Cue
root=Random_1
nodes [
 {id=Random_1, children [WavePlayer_0,WavePlayer_4,WavePlayer_5], properties {weights [1,1,1]}}
 {id=WavePlayer_0, properties {soundWaveAssetPtr=&/Chrono_HvyPistol_Landed_03}}
 {id=WavePlayer_4, properties {soundWaveAssetPtr=&/Chrono_HvyPistol_Landed_01}}
 {id=WavePlayer_5, properties {soundWaveAssetPtr=&/Chrono_HvyPistol_Landed_02}}
]
properties {
 attenuationSettings=&Fort_Weapon_Drop_Att
 maxDistance=3000
 soundClassObject=&Fort_Loot_Collect
 volumeMultiplier=1.3
 concurrencyOverrides {resolutionRule=PreventNew}
 concurrencySet [&Weapon_Drop_StopOldest]
}
