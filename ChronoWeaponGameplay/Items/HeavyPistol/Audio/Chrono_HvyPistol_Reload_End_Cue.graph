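# Sound cue for the end of the Chrono Heavy Pistol reload, classed as weapon foley with reload attenuation.
# A random node picks one of three equally weighted reload-end waves and passes it through a modulator for slight pitch and volume variation.
# A mixer layers that result with a fixed sweetener wave played at half input volume.
# The cue plays at a reduced overall volume, overrides concurrency and has debug enabled.
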SoundCue /ChronoWeaponGameplay/Items/HeavyPistol/Audio/Chrono_HvyPistol_Reload_End_Cue
root=Mixer_0
nodes [
 {id=Mixer_0, children [Modulator_0,WavePlayer_0], properties {inputVolume [1,0.5]}}
 {id=Modulator_0, children [Random_0], properties {pitchMax=1.01, pitchMin=0.99, volumeMax=1.02, volumeMin=0.98}}
 {id=Random_0, children [WavePlayer_3,WavePlayer_4,WavePlayer_5], properties {weights [1,1,1]}}
 {id=WavePlayer_0, properties {soundWaveAssetPtr=&/Chrono_HvyPistol_Reload_End_Sweetener}}
 {id=WavePlayer_3, properties {soundWaveAssetPtr=&/Chrono_HvyPistol_Reload_End_01}}
 {id=WavePlayer_4, properties {soundWaveAssetPtr=&/Chrono_HvyPistol_Reload_End_02}}
 {id=WavePlayer_5, properties {soundWaveAssetPtr=&/Chrono_HvyPistol_Reload_End_03}}
]
properties {
 attenuationSettings=&Fort_Weapon_Reload_Att
 bDebug=true
 bOverrideConcurrency=true
 maxDistance=2250
 soundClassObject=&Fort_WeaponFoley
 volumeMultiplier=0.2
}
