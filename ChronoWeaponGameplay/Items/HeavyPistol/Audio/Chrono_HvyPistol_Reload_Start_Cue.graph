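# Sound cue for the start of the Chrono Heavy Pistol reload, played as weapon foley with reload attenuation.
# A Mixer layers two randomized branches at equal volume: one of three main reload takes and one of six shell takes.
# Each branch passes through its own Modulator for pitch and volume variation, with narrow custom ranges on the main layer and default ranges on the shell layer.
# The cue overrides concurrency and has its debug flag enabled.
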SoundCue /ChronoWeaponGameplay/Items/HeavyPistol/Audio/Chrono_HvyPistol_Reload_Start_Cue
root=Mixer_0
nodes [
 {id=Mixer_0, children [Modulator_0,Modulator_1], properties {inputVolume [1,1]}}
 {id=Modulator_0, children [Random_0], properties {pitchMax=1.01, pitchMin=0.99, volumeMax=1.02, volumeMin=0.98}}
 {id=Modulator_1, children [Random_1]}
 {id=Random_0, children [WavePlayer_0,WavePlayer_1,WavePlayer_2], properties {weights [1,1,1]}}
 {id=Random_1, children [WavePlayer_3,WavePlayer_4,WavePlayer_5,WavePlayer_6,WavePlayer_7,WavePlayer_8], properties {weights [1,1,1,1,1,1]}}
 {id=WavePlayer_0, properties {soundWaveAssetPtr=&/Chrono_HvyPistol_Reload_Start_01}}
 {id=WavePlayer_1, properties {soundWaveAssetPtr=&/Chrono_HvyPistol_Reload_Start_02}}
 {id=WavePlayer_2, properties {soundWaveAssetPtr=&/Chrono_HvyPistol_Reload_Start_03}}
 {id=WavePlayer_3, properties {soundWaveAssetPtr=&/Chrono_HvyPistol_Reload_Start_Shell_01}}
 {id=WavePlayer_4, properties {soundWaveAssetPtr=&/Chrono_HvyPistol_Reload_Start_Shell_02}}
 {id=WavePlayer_5, properties {soundWaveAssetPtr=&/Chrono_HvyPistol_Reload_Start_Shell_03}}
 {id=WavePlayer_6, properties {soundWaveAssetPtr=&/Chrono_HvyPistol_Reload_Start_Shell_04}}
 {id=WavePlayer_7, properties {soundWaveAssetPtr=&/Chrono_HvyPistol_Reload_Start_Shell_05}}
 {id=WavePlayer_8, properties {soundWaveAssetPtr=&/Chrono_HvyPistol_Reload_Start_Shell_06}}
]
properties {
 attenuationSettings=&Fort_Weapon_Reload_Att
 bDebug=true
 bOverrideConcurrency=true
 maxDistance=2250
 soundClassObject=&Fort_WeaponFoley
 volumeMultiplier=0.351
}
